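; Quest interactable with four dance platforms that raise a disco ball while each holds an emoting player.
; Server overlaps start looping timers that check for emotes, with timelines raising or lowering the ball.
; At the top, the server credits the quest to dancing players, skipping credit in Creative, then resets.
; RepNotifies swap floor materials and layer music by dancer count, while clients play confetti and quest reward FX.
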(event EventBeginPlay
  (bind _platform4 (Variables|Default|GetPlatform4))
  (bind _platform3 (Variables|Default|GetPlatform3))
  (bind _platform1 (Variables|Default|GetPlatform1))
  (bind _platform2 (Variables|Default|GetPlatform2))
  ;[1 0.4803 0.03552] Need Better Solution for This!
  (Utilities|FlowControl|Delay 5.0)
  ;[1 1 1] Set Ball Rise Rate via HF number
  (bind _outxy (Fortnite|EvaluateCurveTableRow (Variables|BallMovement|GetBallRaiseTimelineSpeedModifier)))
  (Components|Timeline|SetPlayRate (Variables|BA_QuestInteractable_DancePlatforms|GetRising) _outxy)
  (Networking|SwitchHasAuthority
    (:Authority
      ;[1 1 1] Initialize Overlaps for DanceFloors
      (Utilities|FlowControl|Sequence
        (:then_0
          (Collision|BindEventtoOnComponentBeginOverlap _platform1 (EventDispatchers|CreateEvent :!event Custom|Platform1_Overlap))
          (Collision|BindEventtoOnComponentBeginOverlap _platform2 (EventDispatchers|CreateEvent :!event Custom|Platform2_Overlap))
          (Collision|BindEventtoOnComponentBeginOverlap _platform3 (EventDispatchers|CreateEvent :!event Custom|Platform3_Overlap))
          (Collision|BindEventtoOnComponentBeginOverlap _platform4 (EventDispatchers|CreateEvent :!event Custom|Platform4_Overlap)))
        (:then_1
          (Collision|BindEventtoOnComponentEndOverlap _platform1 (EventDispatchers|CreateEvent :!event Custom|Platform1_EndOverlap))
          (Collision|BindEventtoOnComponentEndOverlap _platform2 (EventDispatchers|CreateEvent :!event Custom|Platform2_EndOverlap))
          (Collision|BindEventtoOnComponentEndOverlap _platform3 (EventDispatchers|CreateEvent :!event Custom|Platform3_EndOverlap))
          (Collision|BindEventtoOnComponentEndOverlap _platform4 (EventDispatchers|CreateEvent :!event Custom|Platform4_EndOverlap)))))
    (:Remote
      ;[1 1 1] Turn on Music on Client to keep from desyncing
      (Components|Activation|Activate (Variables|Default|Getdisco_puzzle_music_Layer2_Cue) true)
      ;[1 1 1] Initialize Quest Progression
      (bind (_questisvalid _questcompleted _objectivecompleted) (Default|AthenaCheckQuestandObjectiveCompletion (Utilities|Array|Get(acopy) (Fortnite|GetLocalFortPlayerControllers true)) (Variables|QuestStuff|GetQuestItem) (Variables|QuestStuff|GetQuestObjBackendName)))
      (if (and _questisvalid (not _objectivecompleted))
        (Variables|QuestStuff|SetIsLocallyActive true)
        (bind _asfort_player_controller_athena (Utilities|Casting|CastToFortPlayerControllerAthena (Utilities|Array|Get(acopy) (Fortnite|GetLocalFortPlayerControllers true))))
        (Quests|BindEventtoOnQuestsUpdated (Quests|GetQuestManager _asfort_player_controller_athena "Invalid") (EventDispatchers|CreateEvent :!event Custom|QuestRewardVisuals%20))))))

(event Custom|UnbindQuestAfterSuccess
  ;[1 1 1] UnbindQuest After Successful update
  (bind _asfort_player_controller_athena (Utilities|Casting|CastToFortPlayerControllerAthena (Utilities|Array|Get(acopy) (Fortnite|GetLocalFortPlayerControllers true))))
  (Quests|UnbindEventfromOnQuestsUpdated (Quests|GetQuestManager _asfort_player_controller_athena "Invalid") (EventDispatchers|CreateEvent :!event Custom|QuestRewardVisuals%20)))

(event Custom|BeginRisingDiscoBall
  ;[1 1 1] Ball Going Up
  (if (not (Variables|BallMovement|GetBallGoingUpAlready))
    (Variables|BallMovement|SetBallLocCurrent (Variables|/Script/Engine.SceneComponent|GetRelativeLocation (Variables|Default|GetDiscoBallBody)))
    (Variables|BallMovement|SetBallGoingUpAlready true)
    (Utilities|FlowControl|Sequence
      (:then_0
        ;[1 1 1] Ball Going Down
        (bind (_direction _movedowntime) (Timeline|Timeline_0 Stop)
          (:Update
            (Transformation|SetRelativeLocation (Variables|Default|GetDiscoBallBody) (Math|Vector|Lerp(Vector) (Variables|BallMovement|GetBallLocCurrent) (+ "0, 0,75.000000" (Variables|/Script/Engine.SceneComponent|GetRelativeLocation (Variables|Default|GetStartingLocation))) _movedowntime) false false))))
      (:then_1
        (bind (_direction_2 _risetime) (Timeline|Rising PlayFromStart)
          (:Update
            (Transformation|SetRelativeLocationAndRotation (Variables|Default|GetDiscoBallBody) (Math|Vector|Lerp(Vector) (Variables|BallMovement|GetBallLocCurrent) (Variables|/Script/Engine.SceneComponent|GetRelativeLocation (Variables|Default|GetEndingLocation)) _risetime) (Math|Rotator|MakeRotator 0.0 0.0 (* _risetime 3.6e+02)) false false))
          (:Finished
            (Development|PrintString "EndPosition" false false)
            (CallFunction|TriggerSuccessCelebration))))
      (:then_2
        (Rendering|Components|Light|SetIntensity (Variables|Default|GetDiscoLights) 1.5e+03)
        (Rendering|Material|SetMaterial (Variables|Default|GetSM_DancePlatformsQuest_Pole) 0 "/Figment_S05/Athena/Items/QuestInteractables/DancePlatforms/FX/MI_DanceFloorQuest.MI_DanceFloorQuest")
        (Components|Activation|Activate (Variables|Default|Getdisco_puzzle_ShepardTone_Cue) false)))))

(event Custom|BeginLoweringDiscoBall
  ;[1 1 1] Ball Going Down
  (if (Variables|BallMovement|GetBallGoingUpAlready)
    (Variables|BallMovement|SetBallLocCurrent (Variables|/Script/Engine.SceneComponent|GetRelativeLocation (Variables|Default|GetDiscoBallBody)))
    (Variables|BallMovement|SetBallGoingUpAlready false)
    (Utilities|FlowControl|Sequence
      (:then_0
        ;[1 1 1] Ball Going Up
        (Timeline|Rising Stop))
      (:then_1
        (Timeline|Timeline_0 PlayFromStart))
      (:then_2
        (Rendering|Components|Light|SetIntensity (Variables|Default|GetDiscoLights) 0.0)
        (Rendering|Material|SetMaterial (Variables|Default|GetSM_DancePlatformsQuest_Pole) 0 "/Game/Athena/Items/QuestInteractables/DancePlatforms/FX/MI_DanceFloorOffQuest.MI_DanceFloorOffQuest")
        (Components|Activation|Deactivate (Variables|Default|Getdisco_puzzle_ShepardTone_Cue))))))

(event Custom|Platform1_Overlap
  (params (OverlappedComponent :type Primitive%20Component%20Object%20Reference) (OtherActor :type Actor%20Object%20Reference) (OtherComp :type Primitive%20Component%20Object%20Reference) (OtherBodyIndex :type Integer) (bFromSweep :type Boolean) (SweepResult :type Hit%20Result%20Structure%20%28by%20ref%29))
  ;[1 1 1] Platform 1
  (bind _checkdanceplattimer (CallFunction|HandlePlatBeginOverlaps OtherActor (Variables|PlatformData|GetPlatform1_Pawns) (Variables|PlatformData|GetPlat1FuncName) (Variables|PlatformData|GetRecheckTime)))
  (Variables|PlatformData|SetCheckDancePlat1Timer _checkdanceplattimer))

(event Custom|Platform2_Overlap
  (params (OverlappedComponent :type Primitive%20Component%20Object%20Reference) (OtherActor :type Actor%20Object%20Reference) (OtherComp :type Primitive%20Component%20Object%20Reference) (OtherBodyIndex :type Integer) (bFromSweep :type Boolean) (SweepResult :type Hit%20Result%20Structure%20%28by%20ref%29))
  ;[1 1 1] Platform 2
  (bind _checkdanceplattimer (CallFunction|HandlePlatBeginOverlaps OtherActor (Variables|PlatformData|GetPlatform2_Pawns) (Variables|PlatformData|GetPlat2FuncName) (Variables|PlatformData|GetRecheckTime)))
  (Variables|PlatformData|SetCheckDancePlat2Timer _checkdanceplattimer))

(event Custom|Platform3_Overlap
  (params (OverlappedComponent :type Primitive%20Component%20Object%20Reference) (OtherActor :type Actor%20Object%20Reference) (OtherComp :type Primitive%20Component%20Object%20Reference) (OtherBodyIndex :type Integer) (bFromSweep :type Boolean) (SweepResult :type Hit%20Result%20Structure%20%28by%20ref%29))
  ;[1 1 1] Platform 3
  (bind _checkdanceplattimer (CallFunction|HandlePlatBeginOverlaps OtherActor (Variables|PlatformData|GetPlatform3_Pawns) (Variables|PlatformData|GetPlat3FuncName) (Variables|PlatformData|GetRecheckTime)))
  (Variables|PlatformData|SetCheckDancePlat3Timer _checkdanceplattimer))

(event Custom|Platform4_Overlap
  (params (OverlappedComponent :type Primitive%20Component%20Object%20Reference) (OtherActor :type Actor%20Object%20Reference) (OtherComp :type Primitive%20Component%20Object%20Reference) (OtherBodyIndex :type Integer) (bFromSweep :type Boolean) (SweepResult :type Hit%20Result%20Structure%20%28by%20ref%29))
  ;[1 1 1] Platform 4
  (bind _checkdanceplattimer (CallFunction|HandlePlatBeginOverlaps OtherActor (Variables|PlatformData|GetPlatform4_Pawns) (Variables|PlatformData|GetPlat4FuncName) (Variables|PlatformData|GetRecheckTime)))
  (Variables|PlatformData|SetCheckDancePlat4Timer _checkdanceplattimer))

(event Custom|Platform1_EndOverlap
  (params (OverlappedComponent :type Primitive%20Component%20Object%20Reference) (OtherActor :type Actor%20Object%20Reference) (OtherComp :type Primitive%20Component%20Object%20Reference) (OtherBodyIndex :type Integer))
  ;[1 1 1] Platform 1
  (bind _platformdancecheck (CallFunction|HandlePlatEndOverlaps OtherActor (Variables|PlatformData|GetPlatform1_Pawns) (Variables|PlatformData|GetCheckDancePlat1Timer)))
  (Variables|PlatformData|SetPlatform1Dancer _platformdancecheck))

(event Custom|Platform2_EndOverlap
  (params (OverlappedComponent :type Primitive%20Component%20Object%20Reference) (OtherActor :type Actor%20Object%20Reference) (OtherComp :type Primitive%20Component%20Object%20Reference) (OtherBodyIndex :type Integer))
  ;[1 1 1] Platform 2
  (bind _platformdancecheck (CallFunction|HandlePlatEndOverlaps OtherActor (Variables|PlatformData|GetPlatform2_Pawns) (Variables|PlatformData|GetCheckDancePlat2Timer)))
  (Variables|PlatformData|SetPlatform2Dancer _platformdancecheck))

(event Custom|Platform3_EndOverlap
  (params (OverlappedComponent :type Primitive%20Component%20Object%20Reference) (OtherActor :type Actor%20Object%20Reference) (OtherComp :type Primitive%20Component%20Object%20Reference) (OtherBodyIndex :type Integer))
  ;[1 1 1] Platform 3
  (bind _platformdancecheck (CallFunction|HandlePlatEndOverlaps OtherActor (Variables|PlatformData|GetPlatform3_Pawns) (Variables|PlatformData|GetCheckDancePlat3Timer)))
  (Variables|PlatformData|SetPlatform3Dancer _platformdancecheck))

(event Custom|Platform4_EndOverlap
  (params (OverlappedComponent :type Primitive%20Component%20Object%20Reference) (OtherActor :type Actor%20Object%20Reference) (OtherComp :type Primitive%20Component%20Object%20Reference) (OtherBodyIndex :type Integer))
  ;[1 1 1] Platform 4
  (bind _platformdancecheck (CallFunction|HandlePlatEndOverlaps OtherActor (Variables|PlatformData|GetPlatform4_Pawns) (Variables|PlatformData|GetCheckDancePlat4Timer)))
  (Variables|PlatformData|SetPlatform4Dancer _platformdancecheck))

(event Custom|TriggerSuccessCelebration
  ;[1 1 1] Success Triggered
  (Variables|BallMovement|SetSuccessBreak true)
  (Networking|SwitchHasAuthority
    (:Authority
      ;[1 1 1] Check if Creative Mode
      (bind _asfort_game_state_athena (Utilities|Casting|CastToFortGameStateAthena (Game|GetGameState)))
      (if (GameplayTags|HasTag (Class|FortGameStateAthena|GetAthenaPlaylistContextTags _asfort_game_state_athena) "(TagName=\"Athena.Creative\")")
        (CallFunction|ResetDiscoBall)
        (else
          ;[1 1 1] Gather Dancing Players and give Quest Reward
          (CallFunction|GatherPlayersforReward)
          (Utilities|FlowControl|Delay 0.5) ; To catch those that start dancing just after completion
          (CallFunction|GatherPlayersforReward)
          (Development|PrintString (Utilities|String|ToString(Integer) (Utilities|Array|Length (Variables|QuestStuff|GetPlayersForReward))) false false)
          (Class|AthenaQuestFunctionLibrary|AthenaGiveQuestUpdatetoPlayers (Variables|QuestStuff|GetQuestItem) (Variables|QuestStuff|GetQuestObjBackendName) (Variables|QuestStuff|GetObjectiveStat) (Variables|QuestStuff|GetPlayersForReward))
          (CallFunction|ResetDiscoBall))))
    (:Remote
      ;[1 1 1] CLIENT ONLY: spawn success FX
      (Utilities|FlowControl|Delay 1.0)
      (Audio|PlaySoundatLocation (Variables|SuccesFX|GetFinishFireworks) (Transformation|GetActorLocation))
      (Effects|SpawnEmitteratLocation (Variables|SuccesFX|GetConfettiEmitter) (Transformation|GetWorldLocation (Variables|Default|GetSuccessPS1)) (Transformation|GetWorldRotation (Variables|Default|GetSuccessPS1)))
      (Utilities|FlowControl|Delay 0.25)
      (Effects|SpawnEmitteratLocation (Variables|SuccesFX|GetConfettiEmitter) (Transformation|GetWorldLocation (Variables|Default|GetSuccessPS2)) (Transformation|GetWorldRotation (Variables|Default|GetSuccessPS2)))
      (Utilities|FlowControl|Delay 0.25)
      (Effects|SpawnEmitteratLocation (Variables|SuccesFX|GetConfettiEmitter) (Transformation|GetWorldLocation (Variables|Default|GetSuccessPS3)) (Transformation|GetWorldRotation (Variables|Default|GetSuccessPS3)))
      (Utilities|FlowControl|Delay 0.25)
      (Effects|SpawnEmitteratLocation (Variables|SuccesFX|GetConfettiEmitter) (Transformation|GetWorldLocation (Variables|Default|GetSuccessPS4)) (Transformation|GetWorldRotation (Variables|Default|GetSuccessPS4))))))

(event Custom|QuestRewardVisuals%20
  ;[1 1 1] CLIENT - Quest Reward FX
  (if (Variables|QuestStuff|GetIsLocallyActive)
    (bind (_questisvalid _questcompleted _objectivecompleted) (Default|AthenaCheckQuestandObjectiveCompletion (Utilities|Array|Get(acopy) (Fortnite|GetLocalFortPlayerControllers true)) (Variables|QuestStuff|GetQuestItem) (Variables|QuestStuff|GetQuestObjBackendName)))
    (if (or _questcompleted _objectivecompleted)
      (Development|PrintString "REWARD DA PEOPLES!!!" false false "(R=1.000000,G=0.552070,B=0.060304,A=1.000000)" 5.0)
      (Effects|SpawnFXSystemAttached (Variables|SuccesFX|GetSuccessEmitter) (Variables|/Script/Engine.Character|GetMesh (Game|Player|GetPlayerPawn (Utilities|Array|Get(acopy) (Fortnite|GetLocalFortPlayerControllers true)))))
      (Audio|PlaySoundatLocation (Variables|SuccesFX|GetSuccessSound) (Transformation|GetActorLocation))
      (Variables|QuestStuff|SetIsLocallyActive true)
      (CallFunction|UnbindQuestAfterSuccess))))

(event Custom|Reset%20Disco%20Ball%20
  ;[1 1 1] Reset Disco Ball
  (Development|PrintString "RESET DISCO BALL" false false)
  (Collision|SetCollisionEnabled (Variables|Default|GetPlatform1) "NoCollision")
  (Utilities|FlowControl|Delay 1.0)
  (Components|Activation|Deactivate (Variables|Default|GetP_DancePlatforms_DiscoBall))
  (Rendering|Components|Light|SetIntensity (Variables|Default|GetDiscoLights) 0.0)
  (Rendering|Material|SetMaterial (Variables|Default|GetSM_DancePlatformsQuest_Pole) 0 "/Game/Athena/Items/QuestInteractables/DancePlatforms/FX/MI_DanceFloorOffQuest.MI_DanceFloorOffQuest")
  (Variables|BallMovement|SetNumPlayersDancing 0)
  (Variables|BallMovement|SetBallGoingUpAlready false)
  (Variables|PlatformData|SetPlatform1Dancer false)
  (Variables|PlatformData|SetPlatform2Dancer false)
  (Variables|PlatformData|SetPlatform3Dancer false)
  (Variables|PlatformData|SetPlatform4Dancer false)
  (Variables|BallMovement|SetBallDropStartLoc (Variables|/Script/Engine.SceneComponent|GetRelativeLocation (Variables|Default|GetDiscoBallBody)))
  (bind (_direction _droptime) (Timeline|BallDrop PlayFromStart)
    (:Update
      (Transformation|SetRelativeLocation (Variables|Default|GetDiscoBallBody) (Math|Vector|Lerp(Vector) (Variables|BallMovement|GetBallDropStartLoc) (+ (Variables|/Script/Engine.SceneComponent|GetRelativeLocation (Variables|Default|GetStartingLocation)) "0, 0,75.000000") _droptime) false false))
    (:Finished
      (CallFunction|BallFinishedResetting))))

(event Custom|BallFinishedResetting
  ;[1 1 1] Turn off FX and reset PlayerPawn arrays
  (Variables|BallMovement|SetSuccessBreak false)
  (Utilities|Array|Clear (Variables|PlatformData|GetPlatform1_Pawns))
  (Utilities|Array|Clear (Variables|PlatformData|GetPlatform2_Pawns))
  (Utilities|Array|Clear (Variables|PlatformData|GetPlatform3_Pawns))
  (Utilities|Array|Clear (Variables|PlatformData|GetPlatform4_Pawns))
  (Utilities|Array|Clear (Variables|QuestStuff|GetCandidatesForReward))
  (Utilities|Array|Clear (Variables|QuestStuff|GetPlayersForReward))
  ;[1 1 1] Turn it all back on again
  (Utilities|FlowControl|Delay 0.5)
  ;;[1 0.585 0.1354] Is this the right setting?
  (Collision|SetCollisionEnabled (Variables|Default|GetPlatform1) "QueryAndPhysics"))

(fn CheckForAll4Dancing
  (if (not (Variables|BallMovement|GetSuccessBreak))
    (Utilities|FlowControl|Sequence
      (:then_0
        (if (Variables|PlatformData|GetPlatform1Dancer)
          (Variables|PlatformData|SetPlat1DancerNum 1)
          (else
            (Variables|PlatformData|SetPlat1DancerNum 0)))
        (if (Variables|PlatformData|GetPlatform2Dancer)
          (Variables|PlatformData|SetPlat2DancerNum 1)
          (else
            (Variables|PlatformData|SetPlat2DancerNum 0)))
        (if (Variables|PlatformData|GetPlatform3Dancer)
          (Variables|PlatformData|SetPlat3DancerNum 1)
          (else
            (Variables|PlatformData|SetPlat3DancerNum 0)))
        (if (Variables|PlatformData|GetPlatform4Dancer)
          (Variables|PlatformData|SetPlat4DancerNum 1)
          (else
            (Variables|PlatformData|SetPlat4DancerNum 0)))
        (Variables|BallMovement|SetNumPlayersDancing (+ (+ (+ (Variables|PlatformData|GetPlat1DancerNum) (Variables|PlatformData|GetPlat2DancerNum)) (Variables|PlatformData|GetPlat3DancerNum)) (Variables|PlatformData|GetPlat4DancerNum))))
      (:then_1
        (if (and (and (and (Variables|PlatformData|GetPlatform1Dancer) (Variables|PlatformData|GetPlatform2Dancer)) (Variables|PlatformData|GetPlatform3Dancer)) (Variables|PlatformData|GetPlatform4Dancer))
          (CallFunction|BeginRisingDiscoBall)
          (else
            (CallFunction|BeginLoweringDiscoBall)))))))

(fn GatherPlayersForReward
  (Networking|SwitchHasAuthority
    (:Authority
      (for _array_element (Variables|PlatformData|GetPlatform1_Pawns)
        (bind _asfort_pawn (Utilities|Casting|CastToFortPawn _array_element))
        (Utilities|Array|AddUnique (Variables|QuestStuff|GetCandidatesForReward) _asfort_pawn))
      (for _array_element_2 (Variables|PlatformData|GetPlatform2_Pawns)
        (bind _asfort_pawn_2 (Utilities|Casting|CastToFortPawn _array_element_2))
        (Utilities|Array|AddUnique (Variables|QuestStuff|GetCandidatesForReward) _asfort_pawn_2))
      (for _array_element_3 (Variables|PlatformData|GetPlatform3_Pawns)
        (bind _asfort_pawn_3 (Utilities|Casting|CastToFortPawn _array_element_3))
        (Utilities|Array|AddUnique (Variables|QuestStuff|GetCandidatesForReward) _asfort_pawn_3))
      (for _array_element_4 (Variables|PlatformData|GetPlatform4_Pawns)
        (bind _asfort_pawn_4 (Utilities|Casting|CastToFortPawn _array_element_4))
        (Utilities|Array|AddUnique (Variables|QuestStuff|GetCandidatesForReward) _asfort_pawn_4))
      (for _array_element_5 (Variables|QuestStuff|GetCandidatesForReward)
        (bind _asfort_player_controller (Utilities|Casting|CastToFortPlayerController (Pawn|GetController _array_element_5))
          (:then
            (if (Emote|IsPlayingEmote _asfort_player_controller)
              (Utilities|Array|AddUnique (Variables|QuestStuff|GetPlayersForReward) _asfort_player_controller)
              (else
                (Development|PrintString "NOT DANCING CANDIDATE" false false "(R=1.000000,G=0.048750,B=0.000000,A=1.000000)" 5.0))))
          (:CastFailed
            (Development|PrintString "CAST FAILED!!!" true true "(R=1.000000,G=0.048750,B=0.000000,A=1.000000)" 5.0))))
      (return))))

(fn HandlePlatBeginOverlaps
  (params (BeginOverlapPawn :type Actor%20Object%20Reference) (PlatformPawns :type Array%20of%20Fort%20Player%20Pawn%20Athena%20Object%20References) (FunctionName :type String) (ReCheckTime :type Float%20%28double-precision%29))
  (if (not (Variables|BallMovement|GetSuccessBreak))
    (bind _asfort_player_pawn_athena (Utilities|Casting|CastToFortPlayerPawnAthena BeginOverlapPawn)
      (:then
        (Utilities|Array|AddUnique PlatformPawns _asfort_player_pawn_athena)
        (bind _returnvalue (Utilities|Time|SetTimerbyFunctionName 0 FunctionName ReCheckTime true))
        (return _returnvalue))
      (:CastFailed
        (Development|PrintString "FAILED CAST")))))

(fn HandlePlatEndOverlaps
  (params (EndOverlapPawn :type Actor%20Object%20Reference) (PlatformPawns :type Array%20of%20Fort%20Player%20Pawn%20Athena%20Object%20References) (CheckDanceTimer :type Timer%20Handle%20Structure))
  (if (not (Variables|BallMovement|GetSuccessBreak))
    (bind _asfort_player_pawn_athena (Utilities|Casting|CastToFortPlayerPawnAthena EndOverlapPawn)
      (:then
        (Utilities|Array|RemoveItem PlatformPawns _asfort_player_pawn_athena)
        (if (not (> (Utilities|Array|Length PlatformPawns) 0))
          (Utilities|Time|ClearandInvalidateTimerbyHandle CheckDanceTimer)
          (return false)))
      (:CastFailed
        (Development|PrintString "FAILED CAST")))))

(fn InternalCheckForDancing
  (params (InPlayers :type Array%20of%20Fort%20Player%20Pawn%20Athena%20Object%20References))
  (bind _array_element (Utilities|Array|ForEachLoopwithBreak (Variables|GetInPlayers))
    (:LoopBody
      (bind _asathena_player_controller (Utilities|Casting|CastToAthena_PlayerController (Pawn|GetController _array_element)))
      (if (Emote|IsPlayingEmote _asathena_player_controller)
        (LocalVariables|SetFoundDancingPawn true)))
    (:Completed
      _))
  (return (Math|Conversions|ToInteger(Boolean) (LocalVariables|GetFoundDancingPawn))))

(fn OnRep_NumPlayersDancing
  (bind _disco_puzzle_music_layer2_cue (Variables|Default|Getdisco_puzzle_music_Layer2_Cue))
  (bind _disco_puzzle_music_layer3_cue (Variables|Default|Getdisco_puzzle_music_Layer3_Cue))
  (bind _disco_puzzle_music_layer4_cue (Variables|Default|Getdisco_puzzle_music_Layer4_Cue))
  (switch int (Variables|BallMovement|GetNumPlayersDancing)
    (:0
      (Networking|SwitchHasAuthority
        (:Remote
          ;[1 1 1] Music Off
          (Audio|Components|Audio|SetVolumeMultiplier _disco_puzzle_music_layer2_cue 0.0 :!class /Script/Engine.AudioComponent))))
    (:1
      (Networking|SwitchHasAuthority
        (:Remote
          ;[1 1 1] Music On
          (Audio|Components|Audio|SetVolumeMultiplier _disco_puzzle_music_layer2_cue 1.0 :!class /Script/Engine.AudioComponent)
          (Audio|Components|Audio|SetVolumeMultiplier _disco_puzzle_music_layer3_cue 0.0 :!class /Script/Engine.AudioComponent))))
    (:2
      (Networking|SwitchHasAuthority
        (:Remote
          (Audio|Components|Audio|SetVolumeMultiplier _disco_puzzle_music_layer3_cue 1.0 :!class /Script/Engine.AudioComponent)
          (Audio|Components|Audio|SetVolumeMultiplier _disco_puzzle_music_layer4_cue 0.0 :!class /Script/Engine.AudioComponent))))
    (:3
      (Networking|SwitchHasAuthority
        (:Remote
          (Audio|Components|Audio|SetVolumeMultiplier _disco_puzzle_music_layer4_cue 1.0 :!class /Script/Engine.AudioComponent))))))

(fn OnRep_Platform1Dancer
  (bind _dancefloor1 (Variables|Default|GetDanceFloor1))
  (if (Variables|PlatformData|GetPlatform1Dancer)
    (Rendering|Material|SetMaterial _dancefloor1 0 "/Figment_S05/Athena/Items/QuestInteractables/DancePlatforms/FX/MI_DanceFloorQuest.MI_DanceFloorQuest")
    (else
      (Rendering|Material|SetMaterial _dancefloor1 0 "/Game/Athena/Items/QuestInteractables/DancePlatforms/FX/MI_DanceFloorOffQuest.MI_DanceFloorOffQuest"))))

(fn OnRep_Platform2Dancer
  (bind _dancefloor2 (Variables|Default|GetDanceFloor2))
  (if (Variables|PlatformData|GetPlatform2Dancer)
    (Rendering|Material|SetMaterial _dancefloor2 0 "/Figment_S05/Athena/Items/QuestInteractables/DancePlatforms/FX/MI_DanceFloorQuest.MI_DanceFloorQuest")
    (else
      (Rendering|Material|SetMaterial _dancefloor2 0 "/Game/Athena/Items/QuestInteractables/DancePlatforms/FX/MI_DanceFloorOffQuest.MI_DanceFloorOffQuest"))))

(fn OnRep_Platform3Dancer
  (bind _dancefloor3 (Variables|Default|GetDanceFloor3))
  (if (Variables|PlatformData|GetPlatform3Dancer)
    (Rendering|Material|SetMaterial _dancefloor3 0 "/Figment_S05/Athena/Items/QuestInteractables/DancePlatforms/FX/MI_DanceFloorQuest.MI_DanceFloorQuest")
    (else
      (Rendering|Material|SetMaterial _dancefloor3 0 "/Game/Athena/Items/QuestInteractables/DancePlatforms/FX/MI_DanceFloorOffQuest.MI_DanceFloorOffQuest"))))

(fn OnRep_Platform4Dancer
  (bind _dancefloor4 (Variables|Default|GetDanceFloor4))
  (if (Variables|PlatformData|GetPlatform4Dancer)
    (Rendering|Material|SetMaterial _dancefloor4 0 "/Figment_S05/Athena/Items/QuestInteractables/DancePlatforms/FX/MI_DanceFloorQuest.MI_DanceFloorQuest")
    (else
      (Rendering|Material|SetMaterial _dancefloor4 0 "/Game/Athena/Items/QuestInteractables/DancePlatforms/FX/MI_DanceFloorOffQuest.MI_DanceFloorOffQuest"))))

(fn Plat1CheckForDancing
  (bind _newnumdancingplayers (CallFunction|InternalCheckforDancing (Variables|PlatformData|GetPlatform1_Pawns)))
  (Variables|PlatformData|SetPlatform1Dancer (>= _newnumdancingplayers 1))
  (CallFunction|CheckforAll4Dancing))

(fn Plat2CheckForDancing
  (bind _newnumdancingplayers (CallFunction|InternalCheckforDancing (Variables|PlatformData|GetPlatform2_Pawns)))
  (Variables|PlatformData|SetPlatform2Dancer (>= _newnumdancingplayers 1))
  (CallFunction|CheckforAll4Dancing))

(fn Plat3CheckForDancing
  (bind _newnumdancingplayers (CallFunction|InternalCheckforDancing (Variables|PlatformData|GetPlatform3_Pawns)))
  (Variables|PlatformData|SetPlatform3Dancer (>= _newnumdancingplayers 1))
  (CallFunction|CheckforAll4Dancing))

(fn Plat4CheckForDancing
  (bind _newnumdancingplayers (CallFunction|InternalCheckforDancing (Variables|PlatformData|GetPlatform4_Pawns)))
  (Variables|PlatformData|SetPlatform4Dancer (>= _newnumdancingplayers 1))
  (CallFunction|CheckforAll4Dancing))

(fn ConstructionScript
  (Transformation|SetRelativeLocation (Variables|Default|GetPlatform1) (Math|Vector|MakeVector (Variables|PlatformData|GetPlatformOffset) (Variables|PlatformData|GetPlatformOffset) 192.0) false false)
  (Transformation|SetRelativeLocation (Variables|Default|GetPlatform2) (Math|Vector|MakeVector (* (Variables|PlatformData|GetPlatformOffset) -1.0) (Variables|PlatformData|GetPlatformOffset) 192.0) false false)
  (Transformation|SetRelativeLocation (Variables|Default|GetPlatform3) (Math|Vector|MakeVector (Variables|PlatformData|GetPlatformOffset) (* (Variables|PlatformData|GetPlatformOffset) -1.0) 192.0) false false)
  (Transformation|SetRelativeLocation (Variables|Default|GetPlatform4) (Math|Vector|MakeVector (* (Variables|PlatformData|GetPlatformOffset) -1.0) (* (Variables|PlatformData|GetPlatformOffset) -1.0) 192.0) false false))
